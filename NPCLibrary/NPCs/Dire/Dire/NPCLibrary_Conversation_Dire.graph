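# Conversation database for the Dire NPC in the NPC library, giving his greeting and access to shared dialogue options.
# The entry point leads to one Speech node with a default pack-themed line and prioritized contextual lines.
# Those lines react to returning after time away, a first-ever conversation, carried Bars, or specific player outfits.
# The speech then links to the shared NPC library Quests and BasicServices conversations by entry tag.
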ConversationDatabase /NPCLibrary/NPCs/Dire/Dire/NPCLibrary_Conversation_Dire
entry_points [{tag=Conversation.Entry.Dire, destinations [#1]}]
nodes [
 ConversationEntryPointNode_7 {
  id=#1
  class=ConversationEntryPointNode
  outputs [#2]
  properties {entryTag {tagName=Conversation.Entry.Dire}}
 }
 Speech_C_4 {
  id=#2
  class=Speech_C
  outputs [#3, #4]
  properties {
   nodeName=""
   generalConfig {
    defaultMessage="Mess with the Alpha, get the pack."
    contextualMessages [
     {
      message=Welcome back to the pack.
      contextRequirements [{requirement {class=FortControllerRequirement_CheckTimeSinceLastPlayedMatch}}]
      priority {curve {rowName=Tandem.Speech.Priority.WhileSincePlayed}}
     }
     {
      message=Ugh. You smell human. Disgusting.
      contextRequirements [{requirement {class=FortControllerRequirement_IsFirstEverConversationWithNPC}}]
      priority {curve {rowName=Tandem.Speech.Priority.FirstEverInteraction}}
     }
     {
      message=I've no need for Bars.
      contextRequirements [{requirement {class=FortControllerRequirement_HasBars}}]
      priority {curve {rowName=Tandem.Speech.Priority.HighTotalBars}}
     }
     {
      message=You look so chompable... Just a little bite? Please?
      requirementMatchPolicy=RequireAny
      contextRequirements [{requirement {
         class=FortControllerRequirement_HasCID
         properties {allowedCIds [
           /Game/Athena/Items/Cosmetics/Characters/CID_069_Athena_Commando_F_PinkBear
           /BRCosmetics/Athena/Items/Cosmetics/Characters/CID_325_Athena_Commando_M_WavyMan
           /BRCosmetics/Athena/Items/Cosmetics/Characters/CID_326_Athena_Commando_F_WavyMan
          ]}
        }}]
      priority {curve {rowName=Tandem.Speech.Priority.OutfitReaction}}
     }
    ]
   }
  }
 }
 ConversationLinkNode_0 {id=#4, class=ConversationLinkNode, properties {remoteEntryTag {tagName=Conversation.Entry.BasicServices}}}
 ConversationLinkNode_1 {id=#3, class=ConversationLinkNode, properties {remoteEntryTag {tagName=Conversation.Entry.NPCLibrary.Quests}}}
]
